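# Parent material for the HornettaVine backpack, layering animated-vine-style emissive glow onto the standard character shader chain (PreFX, CharacterSkinFX, PostFX, cinematic PostFX).
# The desaturated emissive is multiplied by deep veins and noisy glow from the shared HornettaVine body veins function, with time-of-day dimming applied twice.
# A procedural UV mask, one scaled region minus a mirrored, skewed one, limits where the glow replaces the base emissive.
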Material /BRCosmetics/Accessories/FORT_Backpacks/Backpack_F_MED_HornettaVine/Materials/F_MED_HornettaVine_Backpack_Parent
graph {
 properties {
  bAutomaticallySetUsageInEditor=false
  bUseMaterialAttributes=true
  bUsedWithMeshDeformer=true
  bUsedWithMorphTargets=true
  bUsedWithSkeletalMesh=true
  pixelDepthOffsetMode=PDOM_Legacy
  refractionMethod=RM_IndexOfRefraction
  shadingModel=MSM_FromMaterialExpression
  subsurfaceProfile=&SS_HeroSkin_02
  usedShadingModels=Default Lit | Subsurface | Clear Coat | Subsurface Profile | Cloth | Eye
 }
 nodes [
  n1 = MaterialFunctionCall() {materialFunction=&BaseFN_CharacterFunction_PreFX}
  n2 = MaterialFunctionCall(
   AO=@n1.AO
   Distance Blend=@n1.AO
   Fresnel=@n1.AO
   Gmap=@n1.AO
   Pre Skinned Local Position=@n1.AO
   Material=MaterialFunctionCall(
    PreskinnedPosition=@n1.Pre skinned Local Position
    Gmap=@n1.Pre skinned Local Position
    MaterialAttributes=@n1.Pre skinned Local Position
    AO=@n1.Pre skinned Local Position
   ) {materialFunction=&CharacterSkinFX}.Result
  ) {materialFunction=&BaseFN_CharacterFunction_PostFX}
  n3 = GetMaterialAttributes(None=@n2.Result) {attributeGetTypes [#1]}
  n4 = Desaturation(None=@n3.Emissive Color, Fraction=ScalarParameter() {group=HornettaVine, parameterName=Desaturation})
  n5 = ScalarParameter() {defaultValue=0.9, group=HornettaVine, parameterName=TOD_MorningEvening}
  n6 = MaterialFunctionCall(
   Value_Morning=@n5
   Value_Day=Constant() {r=1}
   Value_Evening=@n5
   Value_Night=ScalarParameter() {defaultValue=0.7, group=HornettaVine, parameterName=TOD_Night}
  ) {materialFunction=&Scalar_Time_of_Day}
  n7 = MaterialFunctionCall() {materialFunction=/Game/Characters/Player/Female/Medium/Bodies/F_MED_HornettaVine/Materials/MF_HornettaVine_Veins}
  n8 = Desaturation(None=ScalarParameter() {group=HornettaVine, parameterName=Desaturation_Glow})
  n9 = Multiply(A=@n7.Deep Veins, B=Add(A=Multiply(A=@n8, B=@n7.Glow Gradient With Noise)))
  n10 = Multiply(A=Multiply(A=@n4, B=Multiply(A=@n6.Result, B=@n9)), B=@n6.Result)
  n11 = TextureCoordinate()
  n12 = Constant2Vector() {g=0.15, r=-0.224}
  n13 = MaterialFunctionCall(Float2=Add(A=TextureCoordinate(), B=Constant2Vector() {g=0.25, r=-0.23})) {materialFunction=&BreakOutFloat2Components}
  n14 = Abs(None=Subtract(A=@n13.R) {constB=0.5})
  n15 = MaterialFunctionCall(X=Add(A=OneMinus(None=@n14), B=Multiply(A=@n13.G) {constB=-1.7}), Y=@n13.G) {materialFunction=&MakeFloat2}
  n16 = Constant2Vector() {r=0.425}
  n17 = Subtract(
   A=MaterialFunctionCall(UVs=Add(A=@n11, B=@n12), Texture Scale=Constant2Vector() {g=0.5, r=0.325}) {materialFunction=&ScaleUVsByCenter}.0-1 mask
   B=MaterialFunctionCall(UVs=Add(A=@n15.Result, B=@n16), Texture Scale=Constant2Vector() {g=0.13, r=0.13}) {materialFunction=&ScaleUVsByCenter}.0-1 mask
  )
  MaterialFunctionCall(In=SetMaterialAttributes(MaterialAttributes=@n3.MaterialAttributes, Emissive Color=LinearInterpolate(A=@n4, B=@n10, Alpha=@n17)) {attributeSetTypes [#1]}) {materialFunction=&CINE_MF_CharPostFX}
 ]
}
